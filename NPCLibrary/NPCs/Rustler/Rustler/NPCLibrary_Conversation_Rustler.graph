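# Conversation database for the Rustler NPC, defining the greeting players hear when they start talking to it.
# A single entry point leads to a speech node with a default line plus contextual lines for a first-ever conversation or for specific player character skins.
# Contextual lines are gated by controller requirements, with a curve-driven priority for the first-meeting line.
# The speech then links out to the shared NPC quest and basic services conversations.
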ConversationDatabase /NPCLibrary/NPCs/Rustler/Rustler/NPCLibrary_Conversation_Rustler
entry_points [{tag=Conversation.Entry.Rustler, destinations [#1]}]
nodes [
 ConversationEntryPointNode_7 {
  id=#1
  class=ConversationEntryPointNode
  outputs [#2]
  properties {entryTag {tagName=Conversation.Entry.Rustler}}
 }
 Speech_C_4 {
  id=#2
  class=Speech_C
  outputs [#3, #4]
  properties {
   nodeName=""
   generalConfig {
    defaultMessage="Who knows what chrome'll do to the critters of the island? "
    contextualMessages [
     {
      message="Another day above snakes is another good day, pardner."
      contextRequirements [{requirement {class=FortControllerRequirement_IsFirstEverConversationWithNPC}}]
      priority {curve {rowName=Tandem.Speech.Priority.FirstEverInteraction}}
     }
     {
      message=Is that Chrome on your hands hurting ya?
      contextRequirements [{requirement {class=FortControllerRequirement_HasCID, properties {allowedCIds [&Character_Bites]}}}]
      priority {value=0, curve {curveTable=null, rowName=None}}
     }
     {
      message="Yer a bad egg, I can tell."
      contextRequirements [{requirement {class=FortControllerRequirement_HasCID, properties {allowedCIds [&Character_TheHerald]}}}]
      priority {value=0, curve {curveTable=null, rowName=None}}
     }
     {
      message="Keep an eye on yer horse out there, big sis."
      contextRequirements [{requirement {
         class=FortControllerRequirement_HasCID
         properties {allowedCIds [/BRCosmetics/Athena/Items/Cosmetics/Characters/CID_348_Athena_Commando_F_Medusa]}
        }}]
      priority {value=0, curve {curveTable=null, rowName=None}}
     }
     {
      message=You like skateboards. Ever git up on a bronco?
      contextRequirements [{requirement {class=FortControllerRequirement_HasCID, properties {allowedCIds [&Character_ChillCat]}}}]
      priority {value=0, curve {curveTable=null, rowName=None}}
     }
     {
      message=I loved the episode where y'all learned how to joust on wolves.
      contextRequirements [{requirement {class=FortControllerRequirement_HasCID, properties {allowedCIds [&Character_PinkSpike]}}}]
      priority {value=0, curve {curveTable=null, rowName=None}}
     }
     {
      message="Big claws, small heart. Hate to see it."
      contextRequirements [{requirement {class=FortControllerRequirement_HasCID, properties {allowedCIds [&Character_BadBear]}}}]
      priority {value=0, curve {curveTable=null, rowName=None}}
     }
     {
      message=Stay away from the horses. You'll spook 'em.
      contextRequirements [{requirement {class=FortControllerRequirement_HasCID, properties {allowedCIds [&Character_RoseDust]}}}]
      priority {value=0, curve {curveTable=null, rowName=None}}
     }
    ]
   }
  }
 }
 ConversationLinkNode_0 {id=#4, class=ConversationLinkNode, properties {remoteEntryTag {tagName=Conversation.Entry.BasicServices}}}
 ConversationLinkNode_1 {id=#3, class=ConversationLinkNode, properties {remoteEntryTag {tagName=Conversation.Entry.NPCLibrary.Quests}}}
]
